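# Masked unlit material for the SilentNovel Vane skydiving contrail, used on Niagara ribbons and mesh particles.
# Panned, noise-distorted trail UVs drive opacity and outline masks, combined with gradient and particle-driven dissolves.
# Optional sprinkle textures add color, normals and outlines, while time-of-day functions swap in night tints.
# Thresholded dot products between atmospheric light and normal maps fake stylized two-tone lighting; optional SplineThicken offsets are sine-shaped by vertex color.
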Material /BRCosmetics/Athena/Effects/Particles/Skydiving/NS_Contrail_SilentNovel/M_Contrail_SilentNovel_Vane
graph {
 properties {
  bAutomaticallySetUsageInEditor=false
  bUsedWithNiagaraMeshParticles=true
  bUsedWithNiagaraRibbons=true
  blendMode=BLEND_Masked
  shadingModel=MSM_Unlit
  usedShadingModels=Unlit
 }
 nodes [
  Constant()
  n1 = ScalarParameter() {defaultValue=25, group=SplineThicken, parameterName=SplineThickenWidth, sortPriority=1}
  n2 = Multiply(A=VertexColor().R) {constB=0.5}
  LinearInterpolate(
   A=Constant()
   B=Multiply(
    A=MaterialFunctionCall(WidthBase=@n1, WidthTip=@n1) {materialFunction=&SplineThicken}.WorldPositionOffset
    B=Sine(None=@n2)
   )
   Alpha=ScalarParameter() {group=SplineThicken, parameterName=UseSplineThicken, sortPriority=0}
  )
  n3 = ScalarParameter() {defaultValue=0.1, group=Dissolve, parameterName=GradientlDissolveStart, sortPriority=1}
  n4 = Max(A=Add(A=@n3, B=ScalarParameter() {defaultValue=0.1, group=Dissolve, parameterName=GradialDissolvesoftness, sortPriority=2})) {constB=0.001}
  n5 = TextureCoordinate()
  n6 = Multiply(A=NamedRerouteUsage(), B=ScalarParameter() {defaultValue=0.1, group=Dissolve, parameterName=GradientDissolveDistortion, sortPriority=3})
  n7 = OneMinus(None=ComponentMask(None=Add(A=@n5, B=@n6)) {r=true})
  NamedRerouteDeclaration(None=LinearInterpolate(
    B=SmoothStep(Min=@n3, Max=@n4, Value=@n7) {constMax=0, constMin=0.5}
    Alpha=ScalarParameter() {defaultValue=1, group=Dissolve, parameterName=UseGradientDissolve, sortPriority=0}
   ) {constA=1}) {name=GradientDissolve, nodeColor=<rgba 1 0 0.9844 1>}
  n8 = DynamicParameter() {parameterIndex=1, defaultValue=<rgba 1 0.5 0 1>, paramNames [SplineThicknessMult, Swipe, Dissolve, Param4]}
  n9 = Subtract(A=@n8.Dissolve, B=ScalarParameter() {defaultValue=0.5, group=Dissolve, parameterName=OverallDissolveSoftness, sortPriority=4})
  n10 = Max(A=@n8.Dissolve) {constB=0.001}
  n11 = TextureCoordinate()
  n12 = NamedRerouteUsage()
  n13 = VectorParameter() {group=Dissolve, parameterName=DissolveAndDistortionUV, channelNames {a=PanY, b=PanX, g=ScaleY, r=ScaleX}, defaultValue=<rgba 1 3 0.3 0.2>}
  n14 = AppendVector(A=@n13.PanX, B=@n13.PanX)
  n15 = Add(
   A=AppendVector(A=DynamicParameter() {defaultValue=<rgba 0 0 0 1>, paramNames [UV1 U Offset, Dive, Camera Offset, ""]}.UV1 U Offset, B=Constant())
   B=Multiply(A=Panner(Coordinate=@n11, Time=@n12, Speed=@n14) {speedX=-0.5}, B=AppendVector(A=@n13.ScaleX, B=@n13.ScaleX))
  )
  n16 = TextureSampleParameter2D(UVs=LinearInterpolate(
    A=@n15
    B=MaterialFunctionCall(XY=@n15) {materialFunction=&Swizzle}.YX
    Alpha=ScalarParameter() {group=Trail, parameterName=SwizzleUVs}
   )) {
   group=Dissolve
   parameterName=Distortion Texture
   samplerType=SAMPLERTYPE_LinearGrayscale
   texture=/Game/Effects/Fort_Effects/Textures/Tile/T_CellNoise_Tile_M
  }
  NamedRerouteDeclaration(None=Saturate(None=SmoothStep(Min=@n9, Max=@n10, Value=@n16.R))) {name=Dissolve, nodeColor=<rgba 0 1 0.7422 1>}
  NamedRerouteDeclaration(None=Multiply(A=Time(), B=ScalarParameter() {defaultValue=1, parameterName=TimeMult})) {name=OutTime, nodeColor=<rgba 0 1 0.2031 1>}
  n17 = Add(
   A=TextureCoordinate()
   B=Multiply(A=NamedRerouteUsage(), B=ScalarParameter() {defaultValue=0.3, group=Trail, parameterName=Distortion Strength, sliderMax=1, sliderMin=-1}) {constB=0.1}
  )
  n18 = NamedRerouteUsage()
  n19 = AppendVector(
   A=ScalarParameter() {defaultValue=0.15, group=Trail, parameterName=TrailPanSpeedU, sliderMax=5, sliderMin=-5}
   B=ScalarParameter() {group=Trail, parameterName=TrailPanSpeedV, sliderMax=5, sliderMin=-5}
  )
  n20 = ScalarParameter() {defaultValue=1, group=Trail, parameterName=Scale X, sliderMax=5}
  n21 = Constant() {r=1}
  n22 = MaterialFunctionCall(
   UVs=Multiply(A=Panner(Coordinate=@n17, Time=@n18, Speed=@n19) {speedX=1}, B=AppendVector(A=@n20, B=@n21))
   Texture Scale=MaterialFunctionCall(X=@n21, Y=ScalarParameter() {defaultValue=0.5, group=Trail, parameterName=Scale Y, sliderMax=5}) {materialFunction=&MakeFloat2}.Result
  ) {materialFunction=&ScaleUVsByCenter}
  n23 = MaterialFunctionCall(Float2=@n22.UVs) {materialFunction=&BreakOutFloat2Components}
  n24 = Saturate(None=@n23.G)
  n25 = MaterialFunctionCall(Float2=@n22.UVs) {materialFunction=&BreakOutFloat2Components}
  n26 = Saturate(None=@n25.R)
  n27 = Add(
   A=LinearInterpolate(
    A=MaterialFunctionCall(X=@n23.R, Y=@n24) {materialFunction=&MakeFloat2}.Result
    B=MaterialFunctionCall(X=@n26, Y=@n25.G) {materialFunction=&MakeFloat2}.Result
    Alpha=ScalarParameter() {group=Trail, parameterName=SwizzleUVs}
   )
   B=AppendVector(A=DynamicParameter() {defaultValue=<rgba 0 0 0 1>, paramNames [UV1 U Offset, Dive, Camera Offset, ""]}.UV1 U Offset, B=Constant())
  )
  n28 = LinearInterpolate(
   A=@n27
   B=MaterialFunctionCall(XY=@n27) {materialFunction=&Swizzle}.YX
   Alpha=ScalarParameter() {group=Trail, parameterName=SwizzleUVs}
  )
  n29 = TextureSample(UVs=@n28) {texture=&/T_FX_Contrail_SilentNovel_Vane_Trail}
  NamedRerouteDeclaration(None=@n29.G) {name=TrailOpacity, nodeColor=<rgba 0 1 0.0156 1>}
  n30 = Multiply(A=Multiply(A=NamedRerouteUsage(), B=NamedRerouteUsage()), B=NamedRerouteUsage())
  n31 = ParticleColor()
  NamedRerouteDeclaration(None=Saturate(None=Multiply(A=@n30, B=@n31.A))) {name=OUT_Opacity, nodeColor=<rgba 1 0 0.3047 1>}
  n32 = Constant3Vector()
  n33 = VectorParameter() {group=Color, parameterName=DarkColorMult, defaultValue=<rgba 0.5561 0.557 0.8542 1>}
  n34 = MaterialFunctionCall(
   Color_Morning=@n33.RGB
   Color_Day=@n33.RGB
   Color_Evening=@n33.RGB
   Color_Night=VectorParameter() {group=Color, parameterName=DarkColorMultNight, defaultValue=<rgba 0.8542 0.4449 0.786 1>}.RGB
  ) {materialFunction=&Color_Time_of_Day}
  n35 = NamedRerouteUsage()
  n36 = LinearInterpolate(
   A=@n34.Result
   B=VectorParameter() {group=Color, parameterName=DarkColorMultSprinkles, defaultValue=<rgba 0.4557 0.4565 0.7 1>}.RGB
   Alpha=Saturate(None=Multiply(A=@n35) {constB=2})
  )
  n37 = ScalarParameter() {defaultValue=1, parameterName=UseSprinkles}
  n38 = VectorParameter() {group=Color, parameterName=LightColor, defaultValue=<rgba 0.8 0.12 0.29 1>}
  n39 = MaterialFunctionCall(
   Color_Morning=@n38.RGB
   Color_Day=@n38.RGB
   Color_Evening=@n38.RGB
   Color_Night=VectorParameter() {group=Color, parameterName=LightColorNight, defaultValue=<rgba 0.265 0.09 0.3 1>}.RGB
  ) {materialFunction=&Color_Time_of_Day}
  n40 = NamedRerouteUsage()
  n41 = Multiply(A=@n40, B=VectorParameter() {group=Color, parameterName=SprinklesNightColorMult, defaultValue=<rgba 1 0.5 0.75 1>}.RGB)
  n42 = Multiply(A=NamedRerouteUsage()) {constB=2}
  n43 = LinearInterpolate(
   A=@n39.Result
   B=LinearInterpolate(
    A=@n39.Result
    B=MaterialFunctionCall(Color_Morning=@n40, Color_Day=@n40, Color_Evening=@n40, Color_Night=@n41) {materialFunction=&Color_Time_of_Day}.Result
    Alpha=Saturate(None=@n42)
   )
   Alpha=ScalarParameter() {defaultValue=1, parameterName=UseSprinkles}
  )
  n44 = AtmosphericLightVector()
  n45 = SmoothStep(Value=DotProduct(A=@n44, B=NamedRerouteUsage())) {constMax=0.3, constMin=0.2}
  n46 = SmoothStep(Value=DotProduct(A=@n44, B=NamedRerouteUsage())) {constMax=0.8, constMin=0.55}
  n47 = NamedRerouteUsage()
  n48 = Saturate(None=NamedRerouteUsage())
  n49 = Max(A=SmoothStep(Value=DotProduct(A=@n44, B=@n47)) {constMin=0.1}, B=OneMinus(None=Multiply(A=@n48) {constB=2}))
  n50 = ScalarParameter() {defaultValue=1, parameterName=UseSprinkles}
  n51 = LinearInterpolate(
   A=Multiply(A=LinearInterpolate(A=@n34.Result, B=@n36, Alpha=@n37), B=@n43)
   B=@n43
   Alpha=Multiply(A=Multiply(A=@n45, B=@n46), B=LinearInterpolate(B=@n49, Alpha=@n50) {constA=1})
  )
  n52 = NamedRerouteUsage()
  n53 = ScalarParameter() {defaultValue=1, parameterName=UseSprinkles}
  n54 = OneMinus(None=Add(A=NamedRerouteUsage(), B=LinearInterpolate(B=@n52, Alpha=@n53)))
  NamedRerouteDeclaration(None=MaterialFunctionCall(Emissive=LinearInterpolate(A=@n32, B=@n51, Alpha=@n54)) {materialFunction=&MF_FX_EmissiveAdjustments}.Result) {name=OUT_Emissive, nodeColor=<rgba 0 0.625 1 1>}
  n55 = Add(A=LinearInterpolate(A=ComponentMask(None=@n11) {g=true}, B=ComponentMask(None=@n11) {r=true}, Alpha=ScalarParameter() {group=Trail, parameterName=SwizzleUVs})) {constB=-0.5}
  NamedRerouteDeclaration(None=Multiply(A=@n16.R, B=Multiply(A=@n55) {constB=-1})) {name=Distortion, nodeColor=<rgba 1 0.3083 0.7353 1>}
  NamedRerouteDeclaration(None=TextureSample(UVs=@n28) {samplerType=SAMPLERTYPE_Normal, texture=/Game/Effects/Fort_Effects/Textures/Water/T_Liquid_Tile_01_N}.RGB) {name=TrailNormals, nodeColor=<rgba 1 0.4453 0 1>}
  n56 = TextureCoordinate()
  n57 = NamedRerouteUsage()
  n58 = AppendVector(
   A=ScalarParameter() {defaultValue=0.25, group=Trail, parameterName=NormalNoisePanSpeedU, sliderMax=5, sliderMin=-5}
   B=ScalarParameter() {group=Trail, parameterName=NormalNoisePanSpeedV, sliderMax=5, sliderMin=-5}
  )
  n59 = ScalarParameter() {defaultValue=1, group=Trail, parameterName=Scale X, sliderMax=5}
  n60 = Constant() {r=1}
  n61 = MaterialFunctionCall(
   UVs=Multiply(A=Panner(Coordinate=@n56, Time=@n57, Speed=@n58) {speedX=1}, B=AppendVector(A=@n59, B=@n60))
   Texture Scale=MaterialFunctionCall(X=@n60, Y=ScalarParameter() {defaultValue=0.5, group=Trail, parameterName=Scale Y, sliderMax=5}) {materialFunction=&MakeFloat2}.Result
  ) {materialFunction=&ScaleUVsByCenter}
  n62 = MaterialFunctionCall(Float2=@n61.UVs) {materialFunction=&BreakOutFloat2Components}
  n63 = Saturate(None=@n62.G)
  n64 = MaterialFunctionCall(Float2=@n61.UVs) {materialFunction=&BreakOutFloat2Components}
  n65 = Saturate(None=@n64.R)
  n66 = Add(
   A=LinearInterpolate(
    A=MaterialFunctionCall(X=@n62.R, Y=@n63) {materialFunction=&MakeFloat2}.Result
    B=MaterialFunctionCall(X=@n65, Y=@n64.G) {materialFunction=&MakeFloat2}.Result
    Alpha=ScalarParameter() {group=Trail, parameterName=SwizzleUVs}
   )
   B=AppendVector(A=DynamicParameter() {defaultValue=<rgba 0 0 0 1>, paramNames [UV1 U Offset, Dive, Camera Offset, ""]}.UV1 U Offset, B=Constant())
  )
  n67 = MaterialFunctionCall(XY=@n66) {materialFunction=&Swizzle}
  n68 = ScalarParameter() {group=Trail, parameterName=SwizzleUVs}
  NamedRerouteDeclaration(None=TextureSample(UVs=LinearInterpolate(A=@n66, B=@n67.YX, Alpha=@n68)) {samplerType=SAMPLERTYPE_Normal, texture=/Game/Gadgets/Assets/Telephone_Booth/Textures/T_BlindsNormalNoise_N}.RGB) {name=NormalNoise, nodeColor=<rgba 0 1 0.4609 1>}
  n69 = AppendVector(A=ScalarParameter() {defaultValue=0.1, group=Trail, parameterName=SprinklePanSpeed}, B=Constant())
  n70 = Panner(
   Coordinate=TextureCoordinate()
   Time=NamedRerouteUsage()
   Speed=LinearInterpolate(
    A=@n69
    B=MaterialFunctionCall(XY=@n69) {materialFunction=&Swizzle}.YX
    Alpha=ScalarParameter() {group=Trail, parameterName=SwizzleUVs}
   )
  ) {speedX=1}
  n71 = AppendVector(
   A=Multiply(A=ComponentMask(None=@n70) {r=true}, B=ScalarParameter() {defaultValue=20, group=Trail, parameterName=SprinklesTilingU}) {constB=20}
   B=Multiply(A=ComponentMask(None=@n70) {g=true}, B=ScalarParameter() {defaultValue=2, group=Trail, parameterName=SprinklesTilingV}) {constB=2}
  )
  n72 = LinearInterpolate(
   A=@n71
   B=MaterialFunctionCall(XY=@n71) {materialFunction=&Swizzle}.YX
   Alpha=ScalarParameter() {group=Trail, parameterName=SwizzleUVs}
  )
  n73 = TextureSample(UVs=@n72) {texture=&/T_FX_Contrail_SilentNovel_Vane_Sprinkles_C}
  NamedRerouteDeclaration(None=@n73.RGB) {name=SprinklesColor, nodeColor=<rgba 0 1 0.3672 1>}
  n74 = TextureSample(UVs=@n72) {
   samplerType=SAMPLERTYPE_Normal
   texture=&/T_FX_Contrail_SilentNovel_Vane_Sprinkles_N
  }
  NamedRerouteDeclaration(None=@n74.RGB) {name=SprinklesNormal, nodeColor=<rgba 0 1 0.3438 1>}
  NamedRerouteDeclaration(None=@n73.A) {name=SprinklesAlpha, nodeColor=<rgba 0 0.2734 1 1>}
  NamedRerouteDeclaration(None=@n29.R) {name=OutlineAlpha, nodeColor=<rgba 0 0.1094 1 1>}
  NamedRerouteDeclaration(None=TextureSample(UVs=@n72) {texture=&/T_FX_Contrail_SilentNovel_Vane_Sprinkles_Outline}.R) {name=SprinklesOutline, nodeColor=<rgba 0.5 0 1 1>}
  NamedRerouteUsage()
  NamedRerouteUsage()
  VertexColor()
 ]
}
