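# Niagara skydiving contrail for the SilentNovel cosmetic (Vane variant), trailing donuts, crumbs and sprinkles from two skeletal mesh attachment points.
# Donuts step through four mesh indices as they age, and crumbs spawned from them are culled by parent age into timed bursts.
# Trail_L and Trail_R form spiraling ribbons that sprinkles follow, alongside hand ribbons, orbiting hand sprinkles and faint mesh trails.
# ParaGlideLeanAlpha drives rates and offsets, with separate front-end tuning.
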NiagaraSystem /BRCosmetics/Athena/Effects/Particles/Skydiving/NS_Contrail_SilentNovel/NS_Contrail_SilentNovel_Vane
dependencies {
 usedDataInterfaces [
  &NiagaraDataInterfaceCurve
  &NiagaraDataInterfaceVectorCurve
  &NiagaraDataInterfaceArrayColor
  &NiagaraDataInterfaceParticleRead
  &NiagaraDataInterfaceArrayInt32
  &NiagaraDataInterfaceArrayFloat3
  &NiagaraDataInterfaceSkeletalMesh
 ]
 usedDynamicInputs []
 usedModules [
  &self:SystemSpawnScript.SystemScriptSource.SystemScriptGraph.NiagaraNodeAssignment_1.SetVariables_#1
  &self:SystemSpawnScript.SystemScriptSource.SystemScriptGraph.NiagaraNodeAssignment_2.SetVariables_#2
  &self:SystemSpawnScript.SystemScriptSource.SystemScriptGraph.NiagaraNodeAssignment_3.SetVariables_#3
  &SystemState
  /Game/Effects/Niagara/Modules/Contrails/FN_Contrail_Setup
  &FN_Integrator
  &self:SystemSpawnScript.SystemScriptSource.SystemScriptGraph.NiagaraNodeAssignment_0.SetVariables_#4
 ]
}
emitters {
 Crumbs {
  emitterUpdateScript [
   &EmitterState()
   &SpawnParticlesFromOtherEmitter(
    Attribute Reader=DataInterface() {propertyValues {EmitterBinding {bindingMode=Other, emitterName=Donuts}}}
    Maximum Age Spawn Threshold=0.35
    Minimum Age Spawn Threshold=0.3
    Spawn Rate=&Multiply_Float(A=100, B=$System.CrumbsSpawnRate)
    Spawn Rate Per Particle Cap:NiagaraFloat=&Multiply_Float(A=100, B=$System.CrumbsSpawnRate)
    Use Maximum Age=false
    Use Minimum Age=false
   )
  ]
  particleSpawnScript [
   /SetVariables(Particles.Rand:NiagaraInt32=&RandomRangeInt(Minimum=0, Maximum=1)) {isSetParameters=true}
   &InitializeParticle(
    Color=&LerpLinearColors(
     StartColor=&SelectLinearColorFromArray(
      Array Sampling Mode=::NewEnumerator0
      Direct Array Index=?
      Color Selection Array=DataInterface() {propertyValues {ColorData [
         <rgba 1 0.0437 0.0685 1>
         <rgba 0 0.4621 0.9387 1>
         <rgba 0.007 0.5972 0.2016 1>
         <rgba 1 1 0.8308 1>
         <rgba 0.6445 0.314 0.0931 1>
         <rgba 1 1 0 1>
        ]}}
      Random Seed=0
     )
     EndColor=<rgba 1 1 1 1>
     LerpFactor=&MakeFloatFromInt(Int=$Particles.Rand)
    )
    Lifetime=&Multiply_Float(A=&RandomRangeFloat(Minimum=0.5, Maximum=0.8), B=$System.CrumbsLifeMult)
    Mass Max=2
    Mass Mode=::NewEnumerator3
    Mesh Scale Mode=::NewEnumerator3
    Mesh Uniform Scale=&Multiply_Float(
     A=&Lerp_Float(A=1.5, B=1, Alpha=&MakeFloatFromInt(Int=$Particles.Rand))
     B=$System.CrumbsScale
    )
    Sprite Size Max=<3 14>
    Sprite Size Min=<2 6>
    Sprite Size Mode=::NewEnumerator2
    UsePositionOffset=false
    Write Material Random=true
   )
   &SampleParticlesFromOtherEmitter(Emitter Sampling Mode:ENiagara_EmitterLocSamplingMode=::NewEnumerator0, Particle ID Sampling=::NewEnumerator2)
   &AddVelocity(
    Module.Use Falloff Distance=false
    Override Seed=false
    Velocity Mode=::NewEnumerator1
    Velocity Origin=$Output.SampleParticlesFromOtherEmitter.SampledPosition
    Velocity Speed=&RandomRangeFloat(Minimum=100, Maximum=150)
   ) {module=AddVelocity001}
   &AddVelocity(Velocity=<0 0 1>, Velocity Speed Scale=100) {module=AddVelocity002}
   &ShapeLocation(MaskSpawnGroup=false, Override Seed=false, Sphere Radius=10, UseOwnerScale=true)
   &InitialMeshOrientation(Use Rotation Vector=false)
   /SetVariables(Particles.VisibilityTag=&SelectIntFromArray(Direct Array Index=$Particles.Rand, Int Selection Array=DataInterface() {propertyValues {IntData [0,1]}})) {isSetParameters=true}
   &self:GetParentParticleAge(
    Index:NiagaraInt32=&MakeIntFromFloat(Float=$Particles.SampleParticlesFromOtherEmitter.SampledIndex)
    ParentEmitter=DataInterface() {propertyValues {EmitterBinding {bindingMode=Other, emitterName=Donuts}}}
   ) {module=GetParentParticleAge}
  ]
  particleUpdateScript [
   &ParticleState()
   &self:KillParticlesBasedOnParentParticleAge(
    KillCurve:NiagaraInt32=&MakeIntFromFloat(Float=&FloatFromCurve(
      FloatCurve=DataInterface() {propertyValues {Curve {keys [<at 0 1 RCIM_Constant>,<at 0.24 0 RCIM_Constant>,<at 0.26 1 RCIM_Constant>,<at 0.49 0 RCIM_Constant>,<at 0.51 1 RCIM_Constant>,<at 0.74 0 RCIM_Constant>,<at 0.76 1 RCIM_Constant>,<at 0.92 0 RCIM_Constant>,<at 0.95 1 RCIM_Constant>]}}}
      CurveIndex=$Particles.Initial.ParentNormalizedLife
     ))
    isFirstFrame:NiagaraBool=$Output.ParticleState.FirstFrame
   ) {module=KillParticlesBasedOnParentParticleAge}
   &ScaleMeshSize(Scale Factor=&VectorFromFloat(Value=&FloatFromCurve(FloatCurve=DataInterface() {propertyValues {Curve {
         defaultValue=1
         keys [
          <at 0 0 RCIM_Cubic>
          {arriveTangentWeight=0.0916, interpMode=RCIM_Cubic, leaveTangentWeight=0.281, tangentMode=RCTM_User, time=0.1229, value=1}
          {arriveTangentWeight=0.1257, interpMode=RCIM_Cubic, leaveTangentWeight=0.1667, tangentMode=RCTM_User, time=0.5, value=1}
          {interpMode=RCIM_Cubic, tangentMode=RCTM_User, time=1, value=0}
         ]
        }}})))
   &GravityForce(Gravity=&MakeVector(X=0, Y=0, Z=$System.CrumbsGravity))
   &Drag(Drag=&RandomRangeFloat(Minimum=0.8, Maximum=1.2), Use Linear Drag=true, Use Rotational Drag=false)
   &ScaleColor(
    Scale Alpha=&FloatFromCurve(FloatCurve=DataInterface() {propertyValues {Curve {keys [<at -0.0046 1.0113 RCIM_Cubic>,<at 1.0011 0.002 RCIM_Cubic>]}}})
    ScaleA=true
    ScaleRGB=true
   ) {enabled=false}
   &SolveForcesAndVelocity(Clamp Velocity=false, Limit Acceleration=false)
   &ScaleSpriteSizeBySpeed(
    Max Scale Factor=<1.2 3.3>
    Min Scale Factor=<0.75 0.3>
    Sample Scale Factor By Curve=true
    Scale Factor Curve=DataInterface() {propertyValues {Curve {keys [
        {arriveTangent=2.865, interpMode=RCIM_Cubic, time=0.1247, value=-0.0006}
        {arriveTangent=4.1522, interpMode=RCIM_Cubic, leaveTangent=4.1522, tangentMode=RCTM_User, time=0.3948, value=0.4893}
        <at 0.9958 0.9974 RCIM_Cubic>
       ]}}}
    Velocity Threshold=1500
   )
   /SetVariables(Particles.Scale2:Vector3f=&Multiply_VectorByFloat(Vector=$Particles.Scale, Float=1.2)) {isSetParameters=true}
   /SetVariables(Particles.CameraOffset2=-1) {isSetParameters=true}
   /SetVariables(Particles.Color2=<rgba 0 0 0 1>) {isSetParameters=true}
  ]
  rendererClasses [&NiagaraMeshRendererProperties]
  renderers [
   {rendererClass=&NiagaraMeshRendererProperties, rendererIndex=0}
   {rendererClass=&NiagaraMeshRendererProperties, rendererIndex=1}
   {rendererClass=&NiagaraMeshRendererProperties, rendererIndex=2}
   {rendererClass=&NiagaraMeshRendererProperties, rendererIndex=3}
  ]
 }
 Donuts {
  emitterUpdateScript [
   &EmitterState()
   &SpawnRate(SpawnRate=&Lerp_Float(A=2, B=3, Alpha=$User.ParaGlideLeanAlpha), Use Spawn Probability=false)
  ]
  particleSpawnScript [
   &PartitionParticles()
   &InitializeParticle(
    Color=&FNTimeOfDayColor(Night Color=<rgba 1 0.9667 0.9 1>, Morning Color=<rgba 1 1 1 1>, Day Color=<rgba 1 1 1 1>, Evening Color=<rgba 1 1 1 1>)
    Lifetime=&Multiply_Float(A=$System.DonutLifeMult, B=1)
    Mesh Scale Mode=::NewEnumerator3
    Mesh Uniform Scale=0.5
    UsePositionOffset=false
    Write Material Random=true
   )
   &InitialMeshOrientation(Use Rotation Vector=false)
  ]
  particleUpdateScript [
   &ParticleState()
   &UpdateMeshOrientation(Rotation Vector=&RandomRangeVector(Minimum=<0 0 0>, Maximum=<1 1 1>))
   &SkeletalMeshLocation(
    Bone / Socket Index=$Particles.PartitionParticles.Partition
    Bone Sampling Mode=::NewEnumerator4
    MaskSpawnGroup=false
    Orientation Sampling=::NewEnumerator1
    Sampled Position Offset=&RandomRangeVector(Minimum=<-10 -10 0>, Maximum=<10 10 1>)
    Skeletal Mesh=$System.Skeletal Mesh
   ) {module=SkeletalMeshLocation001}
   /SetVariables(Particles.NewAge:NiagaraFloat=&Multiply_Float(A=$Particles.Age, B=&RandomRangeFloat(Minimum=0.2, Maximum=0.2))) {isSetParameters=true}
   /SetVariables(Particles.Velocity=&Multiply_VectorByFloat(
     Vector=&Lerp_Vector(
      A=<0 0 0>
      B=&Add_Vector(A=&RandomRangeVector(Minimum=<0 0 0>, Maximum=<0 0 0>), B=$System.FN_Contrail_Setup.Direction Velocity)
      Alpha=$Particles.NewAge
     )
     Float=0.5
    )) {isSetParameters=true}
   /SetVariables(Particles.Position=&AddVectorToPosition(
     Position=$Particles.Position
     Vector=&MakeVector(
      X=&FloatFromCurve(
       FloatCurve=DataInterface() {propertyValues {Curve {defaultValue=1, keys [<at 0 0 RCIM_Cubic>,<at 1 1 RCIM_Cubic>]}}}
       Scale Curve=&Lerp_Float(
        A=0
        B=&FN_FrontEndFloatModifier(
         Gameplay=&Lerp_Float(A=80, B=-80, Alpha=&MakeFloatFromInt(Int=$Particles.PartitionParticles.Partition))
         Front End=0
        )
        Alpha=$User.ParaGlideLeanAlpha
       )
      )
      Y=0
      Z=&FloatFromCurve(
       FloatCurve=DataInterface() {propertyValues {Curve {defaultValue=1, keys [{tangentMode=RCTM_User, time=0, value=0}, {tangentMode=RCTM_User, time=1, value=1}]}}}
       Scale Curve=&FN_FrontEndFloatModifier(
        Gameplay=&Lerp_Float(A=220, B=280, Alpha=$User.ParaGlideLeanAlpha)
        Front End=&Lerp_Float(A=160, B=190, Alpha=$User.ParaGlideLeanAlpha)
       )
      )
     )
    )) {isSetParameters=true}
   &ScaleMeshSize(Scale Factor=&VectorFromFloat(Value=&FloatFromCurve(FloatCurve=DataInterface() {propertyValues {Curve {keys [
          <at -0.0251 0 RCIM_Cubic>
          {arriveTangentWeight=0.0589, interpMode=RCIM_Cubic, leaveTangentWeight=0.4711, tangentMode=RCTM_User, time=0.1, value=1.3}
          {arriveTangentWeight=0.4685, interpMode=RCIM_Cubic, leaveTangentWeight=0.0586, tangentMode=RCTM_User, time=0.95, value=1.3105}
          {interpMode=RCIM_Cubic, leaveTangent=0.0124, time=1, value=0}
         ]}}})))
   /SetVariables(Particles.MeshIndex=&MakeIntFromFloat(Float=&FloatFromCurve(FloatCurve=DataInterface() {propertyValues {Curve {keys [<at 0 0 RCIM_Constant>,<at 0.25 1 RCIM_Constant>,<at 0.5 2 RCIM_Constant>,<at 0.75 3 RCIM_Constant>]}}}))) {isSetParameters=true}
   /SetVariables(Particles.Spawn:NiagaraFloat=&FloatFromCurve(FloatCurve=DataInterface() {propertyValues {Curve {keys [<at 0 0 RCIM_Constant>,<at 0.24 0 RCIM_Constant>,<at 0.25 1 RCIM_Constant>,<at 0.26 0 RCIM_Constant>,<at 0.49 0 RCIM_Constant>,<at 0.5 1 RCIM_Constant>,<at 0.51 0 RCIM_Constant>,<at 0.74 0 RCIM_Constant>,<at 0.75 1 RCIM_Constant>,<at 0.76 0 RCIM_Constant>,<at 0.98 1 RCIM_Constant>]}}})) {isSetParameters=true}
   /SetVariables(Particles.Scale2:Vector3f=&Multiply_VectorByFloat(Vector=$Particles.Scale, Float=1.1)) {isSetParameters=true}
   /SetVariables(Particles.CameraOffset2=-5) {isSetParameters=true}
   /SetVariables(Particles.Color2=<rgba 0 0 0 1>) {isSetParameters=true}
  ]
  rendererClasses [&NiagaraMeshRendererProperties]
  renderers [
   {rendererClass=&NiagaraMeshRendererProperties, rendererIndex=0}
   {rendererClass=&NiagaraMeshRendererProperties, rendererIndex=1}
  ]
 }
 FaintTrails {
  emitterUpdateScript [
   &EmitterState(
    Inactive Response=::NewEnumerator1
    Life Cycle Mode=::NewEnumerator1
    Loop Behavior=::NewEnumerator1
    Loop Duration Mode=::NewEnumerator1
    UseLoopDelay=false
   )
   &SpawnRate(
    SpawnRate=&MakeCustomFloatFromBool(
     Bool=&SetBoolByFloatComparison(A=1, B=$User.ParaGlideLeanAlpha, Comparison Type=::NewEnumerator4)
     True Float=5
     False Float=2
    )
    Use Spawn Probability=false
   )
  ]
  particleSpawnScript [
   &PartitionParticles()
   &InitializeParticle(
    Color=<rgba 1 0.2 0.6 1>
    Lifetime=&MakeCustomFloatFromBool(
     Bool=&SetBoolByFloatComparison(A=1, B=$User.ParaGlideLeanAlpha, Comparison Type=::NewEnumerator4)
     True Float=0.7
     False Float=1.25
    )
    Mesh Scale Max=<2 1.5 1.5>
    Mesh Scale Min=<1.2 1.1 1.1>
    Mesh Scale Mode=::NewEnumerator2
    UsePositionOffset=false
    Write Material Random=true
   )
   &InitialMeshOrientation(Mesh Orientation Mode=::NewEnumerator2, Rotation=<0 0.25 0>, Use Rotation Vector=true)
   &InitialMeshOrientation(
    Mesh Orientation Mode=::NewEnumerator2
    Rotation=&RandomRangeVector(Minimum=<0 0 -1>, Maximum=<0 0 1>)
    Use Rotation Vector=true
   ) {module=InitialMeshOrientation001}
   &ShapeLocation(
    Cylinder Height=30
    Cylinder Radius=10
    MaskSpawnGroup=false
    Override Seed=false
    Rotation Mode=::NewEnumerator1
    Shape Primitive=::NewEnumerator1
    Surface Only=false
    UseOwnerScale=true
   )
   &AddVelocity(Velocity=&MakeCustomVectorFromBool(
     Bool=&SetBoolByFloatComparison(A=1, B=$User.ParaGlideLeanAlpha, Comparison Type=::NewEnumerator4)
     True Vector=<0 0 80>
     False Vector=<1 0 30>
    ))
  ]
  particleUpdateScript [
   &ParticleState()
   &ScaleMeshSize(Scale Factor=&Lerp_Vector(A=<1 1 1>, B=<1.5 2 2>, Alpha=$User.ParaGlideLeanAlpha))
   &SkeletalMeshLocation(
    Bone / Socket Index=&SelectIntFromArray(Direct Array Index=$Particles.PartitionParticles.Partition, Int Selection Array=DataInterface() {propertyValues {IntData [0,1]}})
    Bone Sampling Mode=::NewEnumerator4
    MaskSpawnGroup=false
    Orientation Sampling=::NewEnumerator1
    Position Sampling=::NewEnumerator1
    Sampled Position Offset=&AddVectorToPosition(
     Position=Linked() {linkedVariable {name=Particles.Initial.ShapeLocation.ShapePosition, type=Vector3f}}
     Vector=&SelectVectorFromArray(
      Direct Array Index=$Particles.PartitionParticles.Partition
      Vector Selection Array=DataInterface() {propertyValues {FloatData [<0 0 0>, <0 0 0>]}}
     )
    )
    Skeletal Mesh=$System.Skeletal Mesh
   )
   &DynamicMaterialParameters(
    Index 0 Param 3=1
    Index 0 Param 4=&FloatFromCurve(FloatCurve=DataInterface() {propertyValues {Curve {keys [<at 0 0>,<at 1 1>]}}})
    Index 3 Param 1=500
    Module.Index0Param0Write=true
    Module.Index0Param1Write=true
    Module.Index0Param2Write=true
    Module.Index0Param3Write=true
    Module.Index3Param0Write=true
    Module.Index3Param1Write=true
    Module.Index3Param2Write=true
    Module.Index3Param3Write=true
    Write Parameter Index 3=true
   )
   &ScaleColor(
    Scale Alpha=&FloatFromCurve(
     FloatCurve=DataInterface() {propertyValues {Curve {keys [
         {interpMode=RCIM_Cubic, tangentMode=RCTM_User, time=0, value=0}
         {interpMode=RCIM_Cubic, tangentMode=RCTM_User, time=0.4, value=0.8}
         {tangentMode=RCTM_User, time=1, value=0}
        ]}}}
     Scale Curve=&Multiply_Float(
      A=&Lerp_Float(A=0.35, B=0.5, Alpha=$User.ParaGlideLeanAlpha)
      B=&FN_FrontEndFloatModifier(Gameplay=1, Front End=0.8)
     )
    )
    ScaleA=true
    ScaleRGB=true
   )
   &SolveForcesAndVelocity(Clamp Velocity=false, Limit Acceleration=false)
  ]
  rendererClasses [&NiagaraMeshRendererProperties]
  renderers [{rendererClass=&NiagaraMeshRendererProperties, rendererIndex=0}]
 }
 HandRibbon {
  emitterUpdateScript [
   &EmitterState(Life Cycle Mode=::NewEnumerator1, Loop Behavior=::NewEnumerator1, Loop Duration Mode=::NewEnumerator1, UseLoopDelay=false)
   &SpawnBurst_Instantaneous(Spawn Count=2, Use Spawn Probability=false, UseLoopCountLimit=false)
  ]
  particleSpawnScript [
   &InitializeParticle(
    Lifetime Max=1.75
    Lifetime Min=0.8
    Lifetime Mode=::NewEnumerator1
    Mass Max=2
    Mass Mode=::NewEnumerator3
    Mesh Scale Mode=::NewEnumerator3
    Mesh Uniform Scale=0.8
    UsePositionOffset=false
    Write Material Random=true
   )
   &PartitionParticles()
   /SetVariables(Particles.VisibilityTag=$Particles.PartitionParticles.Partition) {isSetParameters=true}
  ]
  particleUpdateScript [
   &ParticleState(Kill Particles When Lifetime Has Elapsed=false, Let Infinitely Lived Particles Die When Emitter Deactivates=true)
   &SkeletalMeshLocation(
    Bone / Socket Index=$Particles.PartitionParticles.Partition
    Bone Sampling Mode=::NewEnumerator4
    MaskSpawnGroup=false
    Orientation Sampling=::NewEnumerator1
    Skeletal Mesh=$System.Skeletal Mesh
   ) {module=SkeletalMeshLocation001}
   &UpdateMeshOrientation(Rotation Rate=-0.15)
   &SolveForcesAndVelocity(Clamp Velocity=false, Limit Acceleration=false)
   /SetVariables(Particles.RotatingMeshOrientation:Quat4f=&MultiplyQuaternion(Quaternion A=$Particles.SkeletalMeshLocation001.SampledRotation, Quaternion B=$Particles.MeshOrientation)) {isSetParameters=true}
  ]
  rendererClasses [&NiagaraMeshRendererProperties]
  renderers [
   {rendererClass=&NiagaraMeshRendererProperties, rendererIndex=0}
   {rendererClass=&NiagaraMeshRendererProperties, rendererIndex=1}
  ]
 }
 HandSprinkles {
  emitterUpdateScript [
   &EmitterState(Life Cycle Mode=::NewEnumerator1, Loop Behavior=::NewEnumerator1, Loop Duration Mode=::NewEnumerator1, UseLoopDelay=false)
   &SpawnRate(SpawnRate=&Lerp_Float(A=10, B=20, Alpha=$User.ParaGlideLeanAlpha), Use Spawn Probability=false)
  ]
  particleSpawnScript [
   &InitializeParticle(
    Color=&SelectLinearColorFromArray(
     Array Sampling Mode=::NewEnumerator0
     Direct Array Index=?
     Color Selection Array=DataInterface() {propertyValues {ColorData [
        <rgba 1 0.0437 0.0685 1>
        <rgba 0 0.4621 0.9387 1>
        <rgba 0.007 0.5972 0.2016 1>
        <rgba 1 1 0.8308 1>
        <rgba 0.6445 0.314 0.0931 1>
        <rgba 1 1 0 1>
       ]}}
     Random Seed=0
    )
    Lifetime Max=1.75
    Lifetime Min=0.8
    Lifetime Mode=::NewEnumerator1
    Mass Max=2
    Mass Mode=::NewEnumerator3
    Mesh Scale Mode=::NewEnumerator3
    Mesh Uniform Scale=&RandomRangeFloat(Minimum=0.8, Maximum=1.2)
    UsePositionOffset=true
    Write Material Random=true
   )
   &PartitionParticles()
   &InitialMeshOrientation(Use Rotation Vector=false)
  ]
  particleUpdateScript [
   &ParticleState()
   &ScaleMeshSize(Scale Factor=&VectorFromFloat(Value=&FloatFromCurve(FloatCurve=DataInterface() {propertyValues {Curve {
         defaultValue=1
         keys [
          {arriveTangent=9.0725, interpMode=RCIM_Cubic, time=0, value=0}
          {arriveTangentWeight=0.0751, interpMode=RCIM_Cubic, leaveTangentWeight=0.3518, tangentMode=RCTM_User, time=0.1408, value=1}
          {arriveTangentWeight=0.3372, interpMode=RCIM_Cubic, leaveTangentWeight=0.1023, tangentMode=RCTM_User, time=0.8, value=1}
          {interpMode=RCIM_Cubic, leaveTangent=-0.0405, time=1, value=0}
         ]
        }}})))
   &SkeletalMeshLocation(
    Bone / Socket Index=$Particles.PartitionParticles.Partition
    Bone Sampling Mode=::NewEnumerator4
    MaskSpawnGroup=false
    Orientation Sampling=::NewEnumerator1
    Position Sampling=::NewEnumerator2
    Skeletal Mesh=$System.Skeletal Mesh
   ) {module=SkeletalMeshLocation001}
   &UpdateMeshOrientation(
    Rotation Rate=&RandomRangeFloat(Minimum=-0.5, Maximum=0.5)
    Rotation Vector=&RandomRangeVector(Minimum=<0 0 0>, Maximum=<1 1 1>)
   )
   &SolveForcesAndVelocity(Clamp Velocity=false, Limit Acceleration=false)
   &RotateAroundPoint(
    Radius:NiagaraFloat=&RandomRangeFloat(Minimum=10, Maximum=15)
    Rotation Center:NiagaraPosition=$Particles.SkeletalMeshLocation001.SampledPosition
    Rotation Phase:NiagaraFloat=&RandomRangeFloat(Minimum=-1, Maximum=1)
    Rotation Rate:NiagaraFloat=&Multiply_Float(A=&RandomRangeFloat(Minimum=0.15, Maximum=0.5), B=-1)
    X Axis:Vector3f=&RotateVectorByQuaternion(Vector To Rotate=<0 1 0>, Quaternion=$Particles.SkeletalMeshLocation001.SampledRotation)
    Y Axis:Vector3f=&RotateVectorByQuaternion(Vector To Rotate=<0 0 1>, Quaternion=$Particles.SkeletalMeshLocation001.SampledRotation)
   )
   /SetVariables(Particles.PositionOffset:NiagaraPosition=&AddVectorToPosition(Position=$Particles.Position, Vector=&RandomRangeVector(Minimum=<0 -5 0>, Maximum=<0 5 0>))) {isSetParameters=true}
   /SetVariables(Particles.Scale2:Vector3f=&Multiply_VectorByFloat(Vector=$Particles.Scale, Float=1.2)) {isSetParameters=true}
   /SetVariables(Particles.CameraOffset2=-1) {isSetParameters=true}
   /SetVariables(Particles.Color2=<rgba 0 0 0 1>) {isSetParameters=true}
  ]
  rendererClasses [&NiagaraMeshRendererProperties]
  renderers [
   {rendererClass=&NiagaraMeshRendererProperties, rendererIndex=0}
   {rendererClass=&NiagaraMeshRendererProperties, rendererIndex=1}
  ]
 }
 Sprinkles {
  emitterUpdateScript [
   &EmitterState(Life Cycle Mode=::NewEnumerator1, Loop Duration=6.7333, UseLoopDelay=false)
   &SpawnRate(SpawnRate=&Lerp_Float(A=16, B=24, Alpha=$User.ParaGlideLeanAlpha), Use Spawn Probability=false)
  ]
  particleSpawnScript [
   &InitializeParticle(
    Color=&SelectLinearColorFromArray(
     Array Sampling Mode=::NewEnumerator0
     Direct Array Index=?
     Color Selection Array=DataInterface() {propertyValues {ColorData [
        <rgba 1 0.0437 0.0685 1>
        <rgba 0 0.4621 0.9387 1>
        <rgba 0.007 0.5972 0.2016 1>
        <rgba 1 1 0.8308 1>
        <rgba 0.6445 0.314 0.0931 1>
        <rgba 1 1 0 1>
       ]}}
     Random Seed=0
    )
    Lifetime=5
    Mass Max=2
    Mass Mode=::NewEnumerator3
    Mesh Scale Mode=::NewEnumerator3
    Mesh Uniform Scale=&RandomRangeFloat(Minimum=0.8, Maximum=1.2)
    Sprite Size Mode=::NewEnumerator3
    Uniform Sprite Size=50
    UsePositionOffset=false
    Write Material Random=true
   )
   &InitialMeshOrientation(Use Rotation Vector=false)
   &PartitionParticles()
  ]
  particleUpdateScript [
   &ParticleState()
   &self:NormalizedRibbonIndex(
    Alpha:NiagaraFloat=&FloatFromCurve(FloatCurve=DataInterface() {propertyValues {Curve {keys [<at 0 1>,<at 1 0>]}}})
    End=24
    Random:NiagaraFloat=&RandomRangeFloat(Minimum=0, Maximum=1)
   ) {module=NormalizedRibbonIndex}
   &ScaleMeshSize(Scale Factor=&Multiply_Vector(
     A=&VectorFromFloat(Value=&FloatFromCurve(
       FloatCurve=DataInterface() {propertyValues {Curve {
          defaultValue=1
          keys [
           {interpMode=RCIM_Cubic, tangentMode=RCTM_User, time=0.1, value=0}
           {arriveTangentWeight=0.0772, interpMode=RCIM_Cubic, leaveTangentWeight=0.3603, tangentMode=RCTM_User, time=0.2, value=1}
           {arriveTangentWeight=0.2333, interpMode=RCIM_Cubic, leaveTangentWeight=0.05, tangentMode=RCTM_User, time=0.8, value=1}
           {tangentMode=RCTM_User, time=0.9, value=0}
          ]
         }}}
       CurveIndex=$Particles.NormalizedIndex
      ))
     B=&VectorFromFloat(Value=&FloatFromCurve(FloatCurve=DataInterface() {propertyValues {Curve {
          defaultValue=1
          keys [
           {interpMode=RCIM_Cubic, tangentMode=RCTM_User, time=0, value=0}
           {arriveTangentWeight=0.0772, interpMode=RCIM_Cubic, leaveTangentWeight=0.3603, tangentMode=RCTM_User, time=0.15, value=1}
           {arriveTangentWeight=0.2333, interpMode=RCIM_Cubic, leaveTangentWeight=0.05, tangentMode=RCTM_User, time=0.85, value=1}
           {tangentMode=RCTM_User, time=1, value=0}
          ]
         }}}))
    ))
   &self:FollowRibbon_02(
    Index:NiagaraInt32=$Particles.CurrentIndex
    Partition:NiagaraInt32=$Particles.PartitionParticles.Partition
    RibbonEmitter_01=DataInterface() {propertyValues {EmitterBinding {bindingMode=Other, emitterName=Trail_L}}}
    RibbonEmitter_02=DataInterface() {propertyValues {EmitterBinding {bindingMode=Other, emitterName=Trail_R}}}
   ) {module=FollowRibbon_02}
   &InterpolateOverTime(
    Rate Of Change:NiagaraFloat=&RandomRangeFloat(Minimum=3, Maximum=5)
    Target Position Value:NiagaraPosition=$Output.FollowRibbon_02.RibbonPos
    Variable Type=::NewEnumerator6
   )
   /SetVariables(Particles.Position=&AddVectorToPosition(
     Position=$Particles.InterpolateOverTime.InterpolateOverTime_Position.Moving Average
     Vector=&Lerp_Vector(
      A=&RandomRangeVector(Minimum=<-20 -20 0>, Maximum=<20 20 0>)
      B=<0 0 0>
      Alpha=&FloatFromCurve(
       FloatCurve=DataInterface() {propertyValues {Curve {
          defaultValue=1
          keys [
           <at 0.1 1>
           {arriveTangent=-2, arriveTangentWeight=0.0333, leaveTangent=-2, leaveTangentWeight=0.2, time=0.4, value=0}
           {arriveTangent=1.669, arriveTangentWeight=0.2, leaveTangent=1.669, leaveTangentWeight=0.3603, time=0.6, value=0}
           <at 0.9991 1>
          ]
         }}}
       CurveIndex=$Particles.NormalizedIndex
      )
     )
    )) {isSetParameters=true}
   /SetVariables(Particles.Scale2:Vector3f=&Multiply_VectorByFloat(Vector=$Particles.Scale, Float=1.2)) {isSetParameters=true}
   /SetVariables(Particles.CameraOffset2=-1) {isSetParameters=true}
   /SetVariables(Particles.Color2=<rgba 0 0 0 1>) {isSetParameters=true}
  ]
  rendererClasses [&NiagaraMeshRendererProperties]
  renderers [
   {rendererClass=&NiagaraMeshRendererProperties, rendererIndex=0}
   {rendererClass=&NiagaraMeshRendererProperties, rendererIndex=1}
  ]
 }
 Trail_L {
  emitterSpawnScript [/SetVariables(Emitter.ParticlePerRibbon:NiagaraInt32=$System.ParticlePerRibbon, Emitter.RibbonNumber:NiagaraInt32=$System.RibbonNumber) {isSetParameters=true, module=SetVariables_#5}]
  emitterUpdateScript [
   &EmitterState(
    Inactive Response=::NewEnumerator1
    Life Cycle Mode=::NewEnumerator1
    Loop Behavior=::NewEnumerator1
    Loop Duration Mode=::NewEnumerator1
    UseLoopDelay=false
   )
   &SpawnBurst_Instantaneous(
    Spawn Count=&Multiply_Int(A=$Emitter.ParticlePerRibbon, B=$Emitter.RibbonNumber)
    Use Spawn Probability=false
    UseLoopCountLimit=false
   )
  ]
  particleSpawnScript [
   &InitializeParticle(Lifetime=5, Ribbon Width=20, Ribbon Width Mode=::NewEnumerator1, UsePositionOffset=false, Write Material Random=true)
   &PartitionParticles(
    ID to Partition:NiagaraInt32=&ReturnExecIndex()
    Number Of Partitions=$Emitter.RibbonNumber
    Particles Per Partition:NiagaraInt32=$Emitter.ParticlePerRibbon
    Partition Mode:ENiagara_PartitionMode_v2=::NewEnumerator1
    Write Partition to Ribbon ID=::NewEnumerator0
   )
   /SetVariables(Particles.NormalizedRibbonPos:NiagaraFloat=&OneMinusFloat(Float=$Particles.PartitionParticles.Normalized Index In Partition)) {isSetParameters=true, module=SetVariables_#6}
   /SetVariables(Particles.SpiralFreq2=0.3) {isSetParameters=true, module=SetVariables_#7}
   /SetVariables(Particles.RibbonU0Override=&OneMinusFloat(Float=$Particles.NormalizedRibbonPos)) {isSetParameters=true, module=SetVariables_#8}
   /SetVariables() {isSetParameters=true, module=SetVariables_#9}
  ]
  particleUpdateScript [
   &ParticleState(Kill Particles When Lifetime Has Elapsed=false)
   &SkeletalMeshLocation(
    Bone Sampling Mode=::NewEnumerator4
    MaskSpawnGroup=false
    Position Sampling=::NewEnumerator2
    Skeletal Mesh=$System.Skeletal Mesh
    Use Sampled Mesh Attribute as Ribbon ID=false
   )
   /SetVariables(Particles.GoalPos2:NiagaraPosition=&AddVectorToPosition(
     Position=$Particles.SkeletalMeshLocation.SampledPosition
     Vector=&Lerp_Vector(
      A=&FN_FrontEndVector3Modifier(
       Gameplay=&VectorFromCurve(
        VectorCurve=DataInterface() {propertyValues {
          XCurve {keys [<at 0 0>,<at 1 0>]}
          YCurve {keys [<at 0 0>,<at 1 0>]}
          ZCurve {keys [
            {arriveTangent=8.632, interpMode=RCIM_Cubic, leaveTangent=8.632, tangentMode=RCTM_User, time=0, value=-5}
            {arriveTangent=23.5865, tangentMode=RCTM_User, time=1, value=180}
           ]}
         }}
        CurveIndex=$Particles.NormalizedRibbonPos
       )
       Front End=&VectorFromCurve(
        VectorCurve=DataInterface() {propertyValues {
          XCurve {keys [<at 0 0>,<at 1 0>]}
          YCurve {keys [<at 0 0>,<at 1 0>]}
          ZCurve {keys [
            {arriveTangent=8.632, interpMode=RCIM_Cubic, leaveTangent=8.632, tangentMode=RCTM_User, time=0, value=-5}
            {arriveTangent=23.5865, tangentMode=RCTM_User, time=1, value=150}
           ]}
         }}
        CurveIndex=$Particles.NormalizedRibbonPos
       )
      )
      B=&FN_FrontEndVector3Modifier(
       Gameplay=&VectorFromCurve(
        VectorCurve=DataInterface() {propertyValues {
          XCurve {keys [<at 0 0 RCIM_Cubic>,<at 1 80 RCIM_Cubic>]}
          YCurve {keys [<at 0 0>,<at 1 0>]}
          ZCurve {keys [
            {arriveTangent=8.632, interpMode=RCIM_Cubic, leaveTangent=8.632, tangentMode=RCTM_User, time=0, value=-5}
            {arriveTangent=23.5865, tangentMode=RCTM_User, time=1, value=210}
           ]}
         }}
        CurveIndex=$Particles.NormalizedRibbonPos
       )
       Front End=&VectorFromCurve(
        VectorCurve=DataInterface() {propertyValues {
          XCurve {keys [<at 0 0 RCIM_Cubic>,<at 1 0 RCIM_Cubic>]}
          YCurve {keys [<at 0 0>,<at 1 0>]}
          ZCurve {keys [
            {arriveTangent=8.632, interpMode=RCIM_Cubic, leaveTangent=8.632, tangentMode=RCTM_User, time=0, value=-5}
            {arriveTangent=23.5865, tangentMode=RCTM_User, time=1, value=180}
           ]}
         }}
        CurveIndex=$Particles.NormalizedRibbonPos
       )
      )
      Alpha=$User.ParaGlideLeanAlpha
     )
    )) {isSetParameters=true, module=SetVariables_#10}
   /SetVariables(Particles.SpiralStrength2:NiagaraFloat=&Multiply_Float(
     A=9
     B=&Lerp_Float(
      A=&FloatFromCurve(
       FloatCurve=DataInterface() {propertyValues {Curve {
          defaultValue=1.625
          keys [
           {arriveTangent=-0.3372, interpMode=RCIM_Cubic, leaveTangent=-0.3372, tangentMode=RCTM_User, time=0.1, value=0}
           {arriveTangentWeight=0.2963, interpMode=RCIM_Cubic, leaveTangentWeight=0.127, tangentMode=RCTM_User, time=0.3, value=2}
           {arriveTangent=-0.3467, interpMode=RCIM_Cubic, leaveTangent=-0.3467, tangentMode=RCTM_User, time=1, value=0}
          ]
         }}}
       CurveIndex=$Particles.NormalizedRibbonPos
      )
      B=&FloatFromCurve(
       FloatCurve=DataInterface() {propertyValues {Curve {
          defaultValue=1.625
          keys [
           {arriveTangent=-0.3372, interpMode=RCIM_Cubic, leaveTangent=-0.3372, tangentMode=RCTM_User, time=0.1, value=0.2}
           {arriveTangentWeight=0.2963, interpMode=RCIM_Cubic, leaveTangentWeight=0.127, tangentMode=RCTM_User, time=0.3, value=2}
           {arriveTangent=-0.3467, interpMode=RCIM_Cubic, leaveTangent=-0.3467, tangentMode=RCTM_User, time=1, value=0}
          ]
         }}}
       CurveIndex=$Particles.NormalizedRibbonPos
      )
      Alpha=$User.ParaGlideLeanAlpha
     )
    )) {isSetParameters=true, module=SetVariables_#11}
   /SetVariables(Particles.SpiralTime2:NiagaraFloat=&Add_Float(
     A=&Add_Float(A=&Multiply_Float(A=$System.Age, B=0.015), B=$Particles.NormalizedRibbonPos)
     B=&Multiply_FloatByInt(
      Float=&Divide_Float(A=$Particles.SpiralFreq2, B=&MakeFloatFromInt(Int=$Emitter.RibbonNumber))
      Integer=$Particles.PartitionParticles.Partition
     )
    )) {isSetParameters=true, module=SetVariables_#12}
   /SetVariables(Particles.SpiralOffset2:Vector3f=&MakeVector(
     X=&Sine(
      Normalized Angle=$Particles.SpiralTime2
      Period=$Particles.SpiralFreq2
      Scale=&Multiply_Float(A=$Particles.SpiralStrength2, B=1.5)
      Bias=0
     )
     Y=&Cosine(
      Normalized Angle=$Particles.SpiralTime2
      Period=$Particles.SpiralFreq2
      Scale=&Multiply_Float(A=$Particles.SpiralStrength2, B=1.5)
      Bias=0
     )
     Z=0
    )) {isSetParameters=true, module=SetVariables_#13}
   /SetVariables(Particles.Position2:NiagaraPosition=&AddVectorToPosition(Position=$Particles.GoalPos2, Vector=$Particles.SpiralOffset2)) {isSetParameters=true, module=SetVariables_#14}
   &SolveForcesAndVelocity(Clamp Velocity=false, Limit Acceleration=false)
   /SetVariables(Particles.RibbonWidth2:NiagaraFloat=&Multiply_Float(
     A=&Multiply_Float(
      A=&FloatFromCurve(
       FloatCurve=DataInterface() {propertyValues {Curve {keys [
           {interpMode=RCIM_Cubic, tangentMode=RCTM_User, time=0.1, value=0}
           {arriveTangent=4, interpMode=RCIM_Cubic, leaveTangent=4, time=0.25, value=1.55}
           {
            arriveTangent=0.788
            arriveTangentWeight=0.2872
            interpMode=RCIM_Cubic
            leaveTangent=0.788
            leaveTangentWeight=0.1567
            tangentMode=RCTM_User
            time=0.6
            value=2
           }
           {arriveTangent=-0.6719, interpMode=RCIM_Cubic, leaveTangent=-0.6719, tangentMode=RCTM_User, time=0.85, value=0}
          ]}}}
       CurveIndex=$Particles.NormalizedRibbonPos
       Scale Curve=18
      )
      B=&Waveform(
       Global Phase=&Add_Float(A=$Emitter.Age, B=$Particles.NormalizedRibbonPos)
       Global Amplitude Scale=2
       "Waveform [1]"=::NewEnumerator12
       "[1] Phase Scale"=1.5
       "[1] Frequency"=1
       "[1] Amplitude Min/Max"=<0.7 1.2>
      )
     )
     B=&FloatFromCurve(
      FloatCurve=DataInterface() {propertyValues {Curve {defaultValue=1, keys [{tangentMode=RCTM_User, time=0, value=0}, {tangentMode=RCTM_User, time=0.5, value=1}]}}}
      CurveIndex=$Emitter.Age
     )
    )) {isSetParameters=true, module=SetVariables_#15}
  ]
  rendererClasses [&NiagaraRibbonRendererProperties]
  renderers [{rendererClass=&NiagaraRibbonRendererProperties}]
 }
 Trail_R {
  emitterSpawnScript [/SetVariables(Emitter.ParticlePerRibbon:NiagaraInt32=$System.ParticlePerRibbon, Emitter.RibbonNumber:NiagaraInt32=$System.RibbonNumber) {isSetParameters=true, module=SetVariables_#5}]
  emitterUpdateScript [
   &EmitterState(
    Inactive Response=::NewEnumerator1
    Life Cycle Mode=::NewEnumerator1
    Loop Behavior=::NewEnumerator1
    Loop Duration Mode=::NewEnumerator1
    UseLoopDelay=false
   )
   &SpawnBurst_Instantaneous(
    Spawn Count=&Multiply_Int(A=$Emitter.ParticlePerRibbon, B=$Emitter.RibbonNumber)
    Use Spawn Probability=false
    UseLoopCountLimit=false
   )
  ]
  particleSpawnScript [
   &InitializeParticle(Lifetime=5, Ribbon Width=20, Ribbon Width Mode=::NewEnumerator1, UsePositionOffset=false, Write Material Random=true)
   &PartitionParticles(
    ID to Partition:NiagaraInt32=&ReturnExecIndex()
    Number Of Partitions=$Emitter.RibbonNumber
    Particles Per Partition:NiagaraInt32=$Emitter.ParticlePerRibbon
    Partition Mode:ENiagara_PartitionMode_v2=::NewEnumerator1
    Write Partition to Ribbon ID=::NewEnumerator0
   )
   /SetVariables(Particles.NormalizedRibbonPos:NiagaraFloat=&OneMinusFloat(Float=$Particles.PartitionParticles.Normalized Index In Partition)) {isSetParameters=true, module=SetVariables_#6}
   /SetVariables(Particles.SpiralFreq2=-0.3) {isSetParameters=true, module=SetVariables_#7}
   /SetVariables(Particles.RibbonU0Override=&OneMinusFloat(Float=$Particles.NormalizedRibbonPos)) {isSetParameters=true, module=SetVariables_#8}
   /SetVariables() {isSetParameters=true, module=SetVariables_#9}
  ]
  particleUpdateScript [
   &ParticleState(Kill Particles When Lifetime Has Elapsed=false)
   &SkeletalMeshLocation(
    Bone / Socket Index=1
    Bone Sampling Mode=::NewEnumerator4
    MaskSpawnGroup=false
    Position Sampling=::NewEnumerator2
    Skeletal Mesh=$System.Skeletal Mesh
    Use Sampled Mesh Attribute as Ribbon ID=false
   )
   /SetVariables(Particles.GoalPos2:NiagaraPosition=&AddVectorToPosition(
     Position=$Particles.SkeletalMeshLocation.SampledPosition
     Vector=&Lerp_Vector(
      A=&FN_FrontEndVector3Modifier(
       Gameplay=&VectorFromCurve(
        VectorCurve=DataInterface() {propertyValues {
          XCurve {keys [<at 0 0>,<at 1 0>]}
          YCurve {keys [<at 0 0>,<at 1 0>]}
          ZCurve {keys [
            {arriveTangent=8.632, interpMode=RCIM_Cubic, leaveTangent=8.632, tangentMode=RCTM_User, time=0, value=-5}
            {arriveTangent=23.5865, tangentMode=RCTM_User, time=1, value=180}
           ]}
         }}
        CurveIndex=$Particles.NormalizedRibbonPos
       )
       Front End=&VectorFromCurve(
        VectorCurve=DataInterface() {propertyValues {
          XCurve {keys [<at 0 0>,<at 1 0>]}
          YCurve {keys [<at 0 0>,<at 1 0>]}
          ZCurve {keys [
            {arriveTangent=8.632, interpMode=RCIM_Cubic, leaveTangent=8.632, tangentMode=RCTM_User, time=0, value=-5}
            {arriveTangent=23.5865, tangentMode=RCTM_User, time=1, value=150}
           ]}
         }}
        CurveIndex=$Particles.NormalizedRibbonPos
       )
      )
      B=&FN_FrontEndVector3Modifier(
       Gameplay=&VectorFromCurve(
        VectorCurve=DataInterface() {propertyValues {
          XCurve {keys [<at 0 0 RCIM_Cubic>,<at 1 -80 RCIM_Cubic>]}
          YCurve {keys [<at 0 0>,<at 1 0>]}
          ZCurve {keys [
            {arriveTangent=8.632, interpMode=RCIM_Cubic, leaveTangent=8.632, tangentMode=RCTM_User, time=0, value=-5}
            {arriveTangent=23.5865, tangentMode=RCTM_User, time=1, value=210}
           ]}
         }}
        CurveIndex=$Particles.NormalizedRibbonPos
       )
       Front End=&VectorFromCurve(
        VectorCurve=DataInterface() {propertyValues {
          XCurve {keys [<at 0 0 RCIM_Cubic>,<at 1 0 RCIM_Cubic>]}
          YCurve {keys [<at 0 0>,<at 1 0>]}
          ZCurve {keys [
            {arriveTangent=8.632, interpMode=RCIM_Cubic, leaveTangent=8.632, tangentMode=RCTM_User, time=0, value=-5}
            {arriveTangent=23.5865, tangentMode=RCTM_User, time=1, value=180}
           ]}
         }}
        CurveIndex=$Particles.NormalizedRibbonPos
       )
      )
      Alpha=$User.ParaGlideLeanAlpha
     )
    )) {isSetParameters=true, module=SetVariables_#10}
   /SetVariables(Particles.SpiralStrength2:NiagaraFloat=&Multiply_Float(
     A=9
     B=&Lerp_Float(
      A=&FloatFromCurve(
       FloatCurve=DataInterface() {propertyValues {Curve {
          defaultValue=1.625
          keys [
           {arriveTangent=-0.3372, interpMode=RCIM_Cubic, leaveTangent=-0.3372, tangentMode=RCTM_User, time=0.1, value=0}
           {arriveTangentWeight=0.2963, interpMode=RCIM_Cubic, leaveTangentWeight=0.127, tangentMode=RCTM_User, time=0.3, value=2}
           {arriveTangent=-0.3467, interpMode=RCIM_Cubic, leaveTangent=-0.3467, tangentMode=RCTM_User, time=1, value=0}
          ]
         }}}
       CurveIndex=$Particles.NormalizedRibbonPos
      )
      B=&FloatFromCurve(
       FloatCurve=DataInterface() {propertyValues {Curve {
          defaultValue=1.625
          keys [
           {arriveTangent=-0.3372, interpMode=RCIM_Cubic, leaveTangent=-0.3372, tangentMode=RCTM_User, time=0.1, value=0.2}
           {arriveTangentWeight=0.2963, interpMode=RCIM_Cubic, leaveTangentWeight=0.127, tangentMode=RCTM_User, time=0.3, value=2}
           {arriveTangent=-0.3467, interpMode=RCIM_Cubic, leaveTangent=-0.3467, tangentMode=RCTM_User, time=1, value=0}
          ]
         }}}
       CurveIndex=$Particles.NormalizedRibbonPos
      )
      Alpha=$User.ParaGlideLeanAlpha
     )
    )) {isSetParameters=true, module=SetVariables_#11}
   /SetVariables(Particles.SpiralTime2:NiagaraFloat=&Add_Float(
     A=&Add_Float(A=&Multiply_Float(A=$System.Age, B=0.015), B=$Particles.NormalizedRibbonPos)
     B=&Multiply_FloatByInt(
      Float=&Divide_Float(A=$Particles.SpiralFreq2, B=&MakeFloatFromInt(Int=$Emitter.RibbonNumber))
      Integer=$Particles.PartitionParticles.Partition
     )
    )) {isSetParameters=true, module=SetVariables_#12}
   /SetVariables(Particles.SpiralOffset2:Vector3f=&MakeVector(
     X=&Sine(
      Normalized Angle=$Particles.SpiralTime2
      Period=$Particles.SpiralFreq2
      Scale=&Multiply_Float(A=$Particles.SpiralStrength2, B=1.5)
      Bias=0
     )
     Y=&Cosine(
      Normalized Angle=$Particles.SpiralTime2
      Period=$Particles.SpiralFreq2
      Scale=&Multiply_Float(A=$Particles.SpiralStrength2, B=1.5)
      Bias=0
     )
     Z=0
    )) {isSetParameters=true, module=SetVariables_#13}
   /SetVariables(Particles.Position2:NiagaraPosition=&AddVectorToPosition(Position=$Particles.GoalPos2, Vector=$Particles.SpiralOffset2)) {isSetParameters=true, module=SetVariables_#14}
   &SolveForcesAndVelocity(Clamp Velocity=false, Limit Acceleration=false)
   /SetVariables(Particles.RibbonWidth2:NiagaraFloat=&Multiply_Float(
     A=&Multiply_Float(
      A=&FloatFromCurve(
       FloatCurve=DataInterface() {propertyValues {Curve {keys [
           {interpMode=RCIM_Cubic, tangentMode=RCTM_User, time=0.1, value=0}
           {arriveTangent=4, interpMode=RCIM_Cubic, leaveTangent=4, time=0.25, value=1.55}
           {
            arriveTangent=0.788
            arriveTangentWeight=0.2872
            interpMode=RCIM_Cubic
            leaveTangent=0.788
            leaveTangentWeight=0.1567
            tangentMode=RCTM_User
            time=0.6
            value=2
           }
           {arriveTangent=-0.6719, interpMode=RCIM_Cubic, leaveTangent=-0.6719, tangentMode=RCTM_User, time=0.85, value=0}
          ]}}}
       CurveIndex=$Particles.NormalizedRibbonPos
       Scale Curve=18
      )
      B=&Waveform(
       Global Phase=&Add_Float(A=$Emitter.Age, B=$Particles.NormalizedRibbonPos)
       Global Amplitude Scale=2
       "Waveform [1]"=::NewEnumerator12
       "[1] Phase Scale"=1.5
       "[1] Frequency"=1
       "[1] Amplitude Min/Max"=<0.7 1.2>
      )
     )
     B=&FloatFromCurve(
      FloatCurve=DataInterface() {propertyValues {Curve {defaultValue=1, keys [{tangentMode=RCTM_User, time=0, value=0}, {tangentMode=RCTM_User, time=0.5, value=1}]}}}
      CurveIndex=$Emitter.Age
     )
    )) {isSetParameters=true, module=SetVariables_#15}
  ]
  rendererClasses [&NiagaraRibbonRendererProperties]
  renderers [{rendererClass=&NiagaraRibbonRendererProperties}]
 }
}
summary {
 systemName=NS_Contrail_SilentNovel_Vane
 emitters [
  {emitterName=Donuts, rendererClasses [&NiagaraMeshRendererProperties]}
  {emitterName=Trail_L, rendererClasses [&NiagaraRibbonRendererProperties]}
  {emitterName=Sprinkles, rendererClasses [&NiagaraMeshRendererProperties]}
  {emitterName=Trail_R, rendererClasses [&NiagaraRibbonRendererProperties]}
  {emitterName=FaintTrails, rendererClasses [&NiagaraMeshRendererProperties]}
  {emitterName=Crumbs, rendererClasses [&NiagaraMeshRendererProperties]}
  {emitterName=HandRibbon, rendererClasses [&NiagaraMeshRendererProperties]}
  {emitterName=HandSprinkles, rendererClasses [&NiagaraMeshRendererProperties]}
 ]
 userVariables [
  User.bIsFrontEnd {type=NiagaraBool}
  User.bIsFrontEndPreview {type=NiagaraBool, defaultValue=true}
  User.ParaGlideLeanAlpha {type=NiagaraFloat, defaultValue=0, description=""}
 ]
}
